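# Material function for the Hyde player character that dissolves the coat through opacity mask as HatAndCoatTransformCurve changes, gated by the IsHydeCoat static switch.
# Bounds-normalized pre-skinned height, offset by the transform value and swirly noise, forms a hard-edged mask.
# Noise is projected with WorldAlignedTexture using pre-skinned position and normal, keeping it fixed to the animated mesh.
# An IsHydeHat world-position-offset branch using a timing curve atlas exists but is not connected to the output.
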MaterialFunction /Game/Characters/Player/Male/Medium/Bodies/M_MED_Hyde/Material/MF_HydeHatandCoat
graph {nodes [
  n1 = VectorParameter() {group=Hyde, parameterName=HatRootLocation, defaultValue=<rgba 0.1432 2.3464 170 1>}
  Add(A=@n1.RGB, B=VectorParameter() {parameterName=HattOffset, defaultValue=<rgba 0 0 5 1>}.RGB)
  Constant()
  n2 = Time()
  Frac(None=@n2)
  n3 = GetMaterialAttributes(None=FunctionInput() {inputName=Material, inputType=FunctionInput_MaterialAttributes}) {attributeGetTypes [#1, #2]}
  n4 = PreSkinnedPosition()
  n5 = FunctionInput(Preview=ShaderStageSwitch(PixelShader=VertexInterpolator(VS=@n4).PS, VertexShader=@n4)) {bUsePreviewValueAsDefault=true, inputName=PreskinnedPosition}
  n6 = Divide(A=@n5, B=PreSkinnedLocalBounds().Half Extents)
  n7 = ScalarParameter() {defaultValue=1, parameterName=HatAndCoatTransformCurve}
  n8 = Subtract(A=Add(A=ComponentMask(None=@n6) {b=true}), B=Multiply(A=OneMinus(None=@n7)) {constB=2})
  n9 = TextureObject() {texture=/Game/Weapons/WeaponSkins/Textures/T_Swirly_Noise_D1}
  n10 = VectorParameter() {group=Hyde, parameterName=NoiseScale, defaultValue=<rgba 64 128 64 1>}
  n11 = FunctionInput(Preview=VertexInterpolator(VS=PreSkinnedNormal()).PS) {bUsePreviewValueAsDefault=true, inputName=PreskinnedNormal}
  n12 = Multiply(
   A=ComponentMask(None=MaterialFunctionCall(TextureObject=@n9, TextureSize=@n10.RGB, WorldPosition=@n5, World Space Normal=@n11) {materialFunction=/Game/Weapons/WeaponSkins/Wraps/MaterialFunctions/WorldAlignedTexture}.XYZ Texture) {r=true}
   B=ScalarParameter() {defaultValue=1, group=Hyde, parameterName=NoiseStrength}
  )
  n13 = Clamp(None=Ceil(None=Subtract(A=@n8, B=@n12)))
  FunctionOutput(None=SetMaterialAttributes!(
    MaterialAttributes=@n3.MaterialAttributes
    OpacityMask=StaticSwitchParameter(True=@n13, False=@n3.OpacityMask) {group=Hyde, parameterName=IsHydeCoat}
   ) {attributeSetTypes [#1]})
  LinearInterpolate(A=Constant() {r=1}, B=@n7, Alpha=ScalarParameter() {defaultValue=1, parameterName=UseTransformCurve})
  n14 = Subtract(A=@n1.RGB, B=WorldPosition().XYZ)
  Multiply(A=@n14) {constB=-1}
  Multiply(A=@n2) {constB=0.25}
  n15 = Time()
  n16 = ScalarParameter() {defaultValue=100000000, group=TargetPractice, parameterName=JH_GameTimeFromBP}
  Saturate(None=Divide(A=Subtract(A=@n15, B=@n16), B=ScalarParameter() {defaultValue=0.5, group=TargetPractice, parameterName=JH_ReactiveDuration}))
  ScalarParameter() {parameterName=JH_TransformCurve}
  n17 = CurveAtlasRowParameter(CurveTime=@n7) {
   atlas=&AthenaCurveAtlas01
   curve=&/HydeTimingCurve
   defaultValue=254
   parameterName=Hyde_TimingCurve
  }
  n18 = ComponentMask(None=@n14) {g=true, r=true}
  n19 = ComponentMask(None=@n14) {b=true}
  StaticSwitchParameter(True=AppendVector(A=Multiply(A=@n17.R, B=@n18) {constA=1}, B=Multiply(A=@n17.G, B=@n19) {constA=1}), False=@n3.WorldPositionOffset) {group=Hyde, parameterName=IsHydeHat}
 ]}
