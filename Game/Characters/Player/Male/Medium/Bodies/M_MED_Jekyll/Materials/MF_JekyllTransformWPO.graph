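# Material function for the Jekyll player body's transformation into Hyde, bulging muscles via World Position Offset and changing skin and shirt color.
# Sphere masks on pre-skinned body points set per-muscle size and time offsets for a sine ripple pushed along vertex normals, reduced on the necktie.
# A curve-atlas timing curve or game-time timer blends in the bulge, darkened shirt and hue-shifted skin.
# IsFace zeroes the offset; IsHyde passes attributes through unchanged.
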MaterialFunction /Game/Characters/Player/Male/Medium/Bodies/M_MED_Jekyll/Materials/MF_JekyllTransformWPO
graph {nodes [
  n1 = VectorParameter() {parameterName=L_BicepLoc, defaultValue=<rgba 20.9107 -0.0001 125.7745 1>}
  Add(A=@n1.RGB, B=ObjectPositionWS())
  n2 = VectorParameter() {parameterName=R_BicepLoc, defaultValue=<rgba -20.6177 0 125.7745 1>}
  Add(A=@n2.RGB, B=ObjectPositionWS())
  n3 = VectorParameter() {parameterName=L_ForeArmLoc, defaultValue=<rgba 31.2882 -0.0002 100.3206 1>}
  Add(A=@n3.RGB, B=ObjectPositionWS())
  n4 = VectorParameter() {parameterName=R_ForeArmLoc, defaultValue=<rgba -31.3804 -0.0001 100.3206 1>}
  Add(A=@n4.RGB, B=ObjectPositionWS())
  n5 = VectorParameter() {parameterName=L_ThighLoc, defaultValue=<rgba 11.338 -0.0001 75.3264 1>}
  Add(A=@n5.RGB, B=ObjectPositionWS())
  n6 = VectorParameter() {parameterName=R_ThighLoc, defaultValue=<rgba -11.9523 -0.0001 76.5302 1>}
  Add(A=@n6.RGB, B=ObjectPositionWS())
  n7 = VectorParameter() {parameterName=L_CalfLoc, defaultValue=<rgba 14.0633 -6.4027 35.6876 1>}
  Add(A=@n7.RGB, B=ObjectPositionWS())
  n8 = VectorParameter() {parameterName=R_CalfLoc, defaultValue=<rgba -14.4292 -6.4026 35.6876 1>}
  Add(A=@n8.RGB, B=ObjectPositionWS())
  Constant3Vector() {constant=<rgba 0.3698 0.3698 0.3698 0>}
  n9 = PreSkinnedPosition()
  n10 = SphereMask(A=@n9, B=@n5.RGB, Radius=ScalarParameter() {defaultValue=24, parameterName=L_ThighScale}, Hardness=Constant())
  n11 = Multiply(A=@n10, B=ScalarParameter() {defaultValue=0.15, parameterName=L_ThighTimeOffset})
  n12 = SphereMask(A=@n9, B=@n6.RGB, Radius=ScalarParameter() {defaultValue=24, parameterName=R_ThighScale}, Hardness=Constant())
  n13 = Multiply(A=@n12, B=ScalarParameter() {defaultValue=0.5, parameterName=R_ThighTimeOffset})
  n14 = SphereMask(A=@n9, B=@n7.RGB, Radius=ScalarParameter() {defaultValue=32, parameterName=L_CalfScale}, Hardness=Constant())
  n15 = ScalarParameter() {defaultValue=0.25, parameterName=L_CalfTimeOffset}
  n16 = SphereMask(A=@n9, B=@n8.RGB, Radius=ScalarParameter() {defaultValue=32, parameterName=R_CalfScale}, Hardness=Constant())
  n17 = Max(
   A=Max(A=Add(A=@n11, B=@n13), B=Multiply(A=@n14, B=@n15))
   B=Multiply(A=@n16, B=ScalarParameter() {defaultValue=0.75, parameterName=R_CalfTimeOffset})
  )
  n18 = SphereMask(A=@n9, B=@n1.RGB, Radius=ScalarParameter() {defaultValue=32, parameterName=L_BicepScale}, Hardness=Constant())
  n19 = SphereMask(A=@n9, B=@n2.RGB, Radius=ScalarParameter() {defaultValue=32, parameterName=R_BicepScale}, Hardness=Constant())
  n20 = ScalarParameter() {defaultValue=0.15, parameterName=R_BicepOffset}
  n21 = SphereMask(A=@n9, B=@n3.RGB, Radius=ScalarParameter() {defaultValue=24, parameterName=L_ForeArmScale}, Hardness=Constant())
  n22 = Multiply(A=@n21, B=ScalarParameter() {defaultValue=0.25, parameterName=L_ForeArmOffset})
  n23 = SphereMask(A=@n9, B=@n4.RGB, Radius=ScalarParameter() {defaultValue=24, parameterName=R_ForeArmScale}, Hardness=Constant())
  n24 = SphereMask(
   A=PreSkinnedPosition()
   B=VectorParameter() {parameterName=L_PecLoc, defaultValue=<rgba 7.0681 0 126.814 1>}.RGB
   Radius=ScalarParameter() {defaultValue=18, parameterName=L_PecScale}
   Hardness=Constant()
  )
  n25 = ScalarParameter() {defaultValue=0.45, parameterName=ChestOffset}
  n26 = SphereMask(
   A=PreSkinnedPosition()
   B=VectorParameter() {parameterName=R_PecLoc, defaultValue=<rgba -6.9199 0 126.814 1>}.RGB
   Radius=ScalarParameter() {defaultValue=32.024, parameterName=R_PecScale}
   Hardness=Constant()
  )
  n27 = Multiply(A=@n26, B=ScalarParameter() {defaultValue=0.5, parameterName=R_PEcOffset})
  n28 = SphereMask(
   A=PreSkinnedPosition()
   B=VectorParameter() {parameterName=HeadLoc, defaultValue=<rgba 0 0 150 1>}.RGB
   Radius=ScalarParameter() {parameterName=HeadSCale}
   Hardness=Constant()
  )
  n29 = Multiply(A=@n28, B=ScalarParameter() {defaultValue=0.15, parameterName=HeadOffset})
  n30 = Max(A=Multiply(A=@n23, B=ScalarParameter() {parameterName=R_ForeArmOffset}), B=Max(A=Multiply(A=@n24, B=@n25), B=Max(A=@n27, B=@n29)))
  n31 = Max(
   A=Multiply(A=@n18, B=ScalarParameter() {defaultValue=0.35, parameterName=L_BicepOffset})
   B=Max(A=Multiply(A=@n19, B=@n20), B=Max(A=@n22, B=@n30))
  )
  n32 = Multiply(A=Add(A=Time(), B=Max(A=@n17, B=@n31)), B=ScalarParameter() {defaultValue=1.5, parameterName=DebugTimeSpeed})
  Frac(None=@n32)
  Frac(None=Time())
  n33 = LinearInterpolate(A=ConstantBiasScale(None=Sine(None=@n32)), B=Constant() {r=1}, Alpha=ScalarParameter() {parameterName=DisableRipple})
  n34 = Multiply(A=@n10, B=ScalarParameter() {defaultValue=7, parameterName=L_ThighWPOSCALE})
  n35 = Multiply(A=@n12, B=ScalarParameter() {defaultValue=7, parameterName=R_ThighWPOSCALE})
  n36 = ScalarParameter() {defaultValue=6, parameterName=L_CalfWPOSCALE}
  n37 = Max(
   A=Max(A=Max(A=@n34, B=@n35), B=Multiply(A=@n14, B=@n36))
   B=Multiply(A=@n16, B=ScalarParameter() {defaultValue=6, parameterName=R_CalfWPOSCALE})
  )
  n38 = Multiply(A=@n18, B=ScalarParameter() {defaultValue=6, parameterName=ArmWPOSCALE})
  n39 = ScalarParameter() {parameterName=ArmWPOSCALE}
  n40 = Max(
   A=Max(A=Max(A=@n37, B=@n38), B=Multiply(A=@n19, B=@n39))
   B=Multiply(A=@n21, B=ScalarParameter() {defaultValue=7, parameterName=L_ForeArmWPOSCALE})
  )
  n41 = Multiply(A=@n23, B=ScalarParameter() {defaultValue=7, parameterName=R_foreArmWPOSCALE})
  n42 = Multiply(A=ScalarParameter() {defaultValue=7, parameterName=L_PecWPOSCALE}, B=@n24)
  n43 = Max(
   A=Multiply(A=ScalarParameter() {defaultValue=7, parameterName=R_PecWPOSCALE}, B=@n26)
   B=Multiply(A=ScalarParameter() {parameterName=HeadWPOSCALE}, B=@n28)
  )
  n44 = Multiply(A=Max(A=Max(A=@n40, B=@n41), B=Max(A=@n42, B=@n43)), B=ScalarParameter() {defaultValue=1, parameterName=Global_WPO_ScaleFactor})
  n45 = TextureSampleParameter2D() {parameterName=JH_BodyMask, texture=/Game/Characters/Player/Male/Medium/Bodies/M_MED_Jekyll/Textures/T__MED_Jekyll_Body_FX_2}
  n46 = LinearInterpolate(A=@n44, B=Multiply(A=@n44, B=ScalarParameter() {defaultValue=0.2, parameterName=JH_NeckTie_ScaleFactor}), Alpha=@n45.G)
  n47 = ScalarParameter() {parameterName=JH_TransformCurve}
  n48 = Multiply(A=@n47) {constB=0.75}
  n49 = Multiply(A=VertexNormalWS(), B=Clamp(None=Multiply(A=@n33, B=@n46) {constA=1}, Min=Multiply(A=@n46, B=@n48)) {maxDefault=12})
  n50 = GetMaterialAttributes(None=FunctionInput() {inputName=Material, inputType=FunctionInput_MaterialAttributes}) {attributeGetTypes [#1, #2]}
  n51 = Subtract(A=Time(), B=ScalarParameter() {defaultValue=100000000, group=TargetPractice, parameterName=JH_GameTimeFromBP})
  n52 = ScalarParameter() {defaultValue=0.5, group=TargetPractice, parameterName=JH_ReactiveDuration}
  n53 = Saturate(None=Divide(A=@n51, B=@n52))
  n54 = CurveAtlasRowParameter(CurveTime=@n47) {
   atlas=&AthenaCurveAtlas01
   curve=/Game/Characters/Player/Male/Medium/Bodies/M_MED_Hyde/Material/HydeTimingCurve
   defaultValue=254
   parameterName=JekyllTimingCurve
  }
  n55 = ScalarParameter() {defaultValue=1, parameterName=UseTransformCurve}
  n56 = StaticSwitchParameter(
   True=Constant()
   False=LinearInterpolate(A=@n50.WorldPositionOffset, B=Add(A=@n50.WorldPositionOffset, B=@n49), Alpha=LinearInterpolate(A=@n53, B=@n54.A, Alpha=@n55))
  ) {parameterName=IsFace}
  n57 = ScalarParameter() {defaultValue=0.1, parameterName=JH_ShirtDarken}
  n58 = Subtract(A=@n51, B=ScalarParameter() {defaultValue=0.5, parameterName=JH_ColorChangeDelay})
  n59 = LinearInterpolate(A=Saturate(None=Divide(A=@n58, B=@n52)), B=@n54.B, Alpha=@n55)
  n60 = MaterialFunctionCall(Hue Shift Percentage=Constant() {r=0.55}, Texture=@n50.BaseColor) {materialFunction=&HueShift}
  n61 = Constant() {r=0.2}
  n62 = LinearInterpolate(
   A=LinearInterpolate(A=@n50.BaseColor, B=Multiply(A=@n50.BaseColor, B=@n57), Alpha=Multiply(A=@n45.R, B=@n59))
   B=LinearInterpolate(A=@n50.BaseColor, B=Multiply(A=@n60.Result, B=@n61), Alpha=@n59)
   Alpha=FunctionInput() {inputName=SkinMask, inputType=FunctionInput_Scalar}
  )
  FunctionOutput(None=StaticSwitchParameter(
    True=@n50.MaterialAttributes
    False=SetMaterialAttributes!(MaterialAttributes=@n50.MaterialAttributes, WorldPositionOffset=@n56, BaseColor=@n62) {attributeSetTypes [#1, #2]}
   ) {parameterName=IsHyde})
  MaterialFunctionCall(Texture=@n50.BaseColor) {materialFunction=&HueShift}
  Multiply()
  Normalize(VectorInput=Subtract(A=WorldPosition().XYZ, B=ObjectPositionWS()))
  ScalarParameter() {parameterName=JH_TransformCurve}
  Subtract(
   A=@n5.RGB
   B=MaterialFunctionCall() {materialFunction=&ObjectPivotPoint}.Object Pivot Location
  )
  Subtract(
   A=@n6.RGB
   B=MaterialFunctionCall() {materialFunction=&ObjectPivotPoint}.Object Pivot Location
  )
  Subtract(
   A=@n7.RGB
   B=MaterialFunctionCall() {materialFunction=&ObjectPivotPoint}.Object Pivot Location
  )
  Subtract(
   A=@n8.RGB
   B=MaterialFunctionCall() {materialFunction=&ObjectPivotPoint}.Object Pivot Location
  )
  Subtract(
   A=@n1.RGB
   B=MaterialFunctionCall() {materialFunction=&ObjectPivotPoint}.Object Pivot Location
  )
  Subtract(
   A=@n2.RGB
   B=MaterialFunctionCall() {materialFunction=&ObjectPivotPoint}.Object Pivot Location
  )
  Subtract(
   A=@n3.RGB
   B=MaterialFunctionCall() {materialFunction=&ObjectPivotPoint}.Object Pivot Location
  )
  Subtract(
   A=@n4.RGB
   B=MaterialFunctionCall() {materialFunction=&ObjectPivotPoint}.Object Pivot Location
  )
  TextureSample() {texture=/Game/Characters/Player/Male/Medium/Bodies/M_MED_Jekyll/Textures/T_M_MED_Jekyll_face_D}
  VectorParameter() {parameterName=HydeSkinColor, defaultValue=<rgba 0.0578 0.162 0.2384 1>}
  WorldPosition()
  WorldPosition()
  WorldPosition()
  WorldPosition()
  WorldPosition()
  WorldPosition()
  WorldPosition()
  WorldPosition()
  WorldPosition()
 ]}
