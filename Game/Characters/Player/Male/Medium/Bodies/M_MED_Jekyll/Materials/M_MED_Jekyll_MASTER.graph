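# Master character material for the M_MED_Jekyll player body, using material attributes with Default Lit, Subsurface Profile and Cloth shading.
# Base character PreFX feeds switchable panning emissive, cloth fuzz and silk layers, masked by channels of one SkinFX_Mask texture, then emissive fresnel.
# A skin-masked Jekyll transform WPO function and the Hyde hat-and-coat function, fed pre-skinned position and interpolated pre-skinned normal, follow before the base PostFX function.
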Material /Game/Characters/Player/Male/Medium/Bodies/M_MED_Jekyll/Materials/M_MED_Jekyll_MASTER
graph {
 properties {
  bUseMaterialAttributes=true
  bUsedWithMeshDeformer=true
  bUsedWithMorphTargets=true
  bUsedWithSkeletalMesh=true
  pixelDepthOffsetMode=PDOM_Legacy
  shadingModel=MSM_FromMaterialExpression
  subsurfaceProfile=&SS_HeroSkin_02
  usedShadingModels=Default Lit | Subsurface Profile | Cloth
 }
 nodes [
  n1 = MaterialFunctionCall() {materialFunction=&BaseFN_CharacterFunction_PreFX}
  n2 = VertexInterpolator(VS=PreSkinnedNormal())
  n3 = TextureSample(Tex=TextureObjectParameter() {group=Skin FX, parameterName=SkinFX_Mask})
  n4 = ChannelMaskParameter(None=@n3.RGBA) {group=Skin FX, parameterName=Silk Channel}
  n5 = StaticSwitchParameter(
   True=MaterialFunctionCall!(
    Alpha=StaticComponentMaskParameter(None=@n3.RGBA) {defaultR=true, group=Skin FX, parameterName=PanningEmissive_Channel}
    isVehicle
    Gmap=@n1.Gmap
    UVOffset=StaticComponentMaskParameter(None=@n3.RGBA) {defaultB=true, group=Skin FX, parameterName=PanningEmissive_UVOffset_Channel}
    Local Normal=@n2.PS
    Local Position=@n1.Gmap
    EmissiveColorOverride
    Material=@n1.Gmap
   ) {materialFunction=/Game/Weapons/WeaponSkins/MaterialFunction/PanningEmissive}.Result
   False=@n1.Result
  ) {group=Skin FX, parameterName=Use Panning Emissive}
  n6 = StaticSwitchParameter(
   True=MaterialFunctionCall!(Alpha=StaticComponentMaskParameter(None=@n3.RGBA) {defaultR=true, group=Skin FX, parameterName=Cloth Channel}, Gmap=@n1.Gmap, Material=@n5) {materialFunction=/Game/Weapons/WeaponSkins/MaterialFunction/ClothFuzz}.Result
   False=@n5
  ) {group=Skin FX, parameterName=Use Cloth fuzz}
  n7 = MaterialFunctionCall!(In=StaticSwitchParameter(
    True=MaterialFunctionCall!(Alpha=@n4.R, Material=@n6) {materialFunction=/Game/Weapons/WeaponSkins/MaterialFunction/MF_Silk}.Result
    False=@n6
   ) {group=Skin FX, parameterName=Use Silk}) {materialFunction=/Game/Weapons/WeaponSkins/MaterialFunction/MF_EmissiveFresnel}
  MaterialFunctionCall(
   AO=@n1.AO
   Distance Blend=@n1.AO
   Fresnel=@n1.AO
   Gmap=@n1.AO
   Pre Skinned Local Position=@n1.AO
   Material=MaterialFunctionCall!(
    PreskinnedPosition=@n1.Pre skinned Local Position
    PreskinnedNormal=@n2.PS
    Material=MaterialFunctionCall!(SkinMask=@n1.Skin, Material=@n7.Result) {materialFunction=&/MF_JekyllTransformWPO}.Result
   ) {materialFunction=/Game/Characters/Player/Male/Medium/Bodies/M_MED_Hyde/Material/MF_HydeHatandCoat}.Result
  ) {materialFunction=&BaseFN_CharacterFunction_PostFX}
 ]
}
